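# Physics asset for the necklace on the FirstClassSecond medium female player cosmetic body, providing capsule bodies and joints for its dynamic bones.
# Capsules on the upper spine and both clavicles approximate the torso, with the clavicles and necklace base joined to spine_05 by free angular constraints.
# A dynamic chain hangs from the base through three bones, with a collar branch off the middle one, using tight swing limits and zero-degree twist limits.
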PhysicsAsset /BRCosmetics/Characters/Player/Female/Medium/Bodies/F_MED_FirstClassSecond/Meshes/F_MED_Necklace_FirstClassSecond
bodies {
 base_necklace_00 [{shapeType=Capsule, length=0.505, radius=0.505, shapeName=base_necklace_00_capsule, rotation=<rot 0 -90 0>}]
 clavicle_l [{shapeType=Capsule, length=4.8, radius=4.8, shapeName=clavicle_l_capsule, center=<0 0.9 -2>, rotation=<rot 90 0 0>}]
 clavicle_r [{shapeType=Capsule, length=4.8, radius=4.8, shapeName=clavicle_r_capsule, center=<0 -0.9 2>, rotation=<rot 90 0 0>}]
 dyn_necklace_00 [{shapeType=Capsule, length=9.5585, radius=1.3, shapeName=dyn_necklace_00_capsule, center=<4 0 0>}]
 dyn_necklace_01 [{shapeType=Capsule, length=9.5585, radius=1.1, shapeName=dyn_necklace_01_capsule, center=<2 0 0>}]
 dyn_necklace_01_collar [{shapeType=Capsule, length=9.5585, radius=1, shapeName=dyn_necklace_01_collar_capsule, center=<2 0 0>}]
 dyn_necklace_02 [{
   shapeType=Capsule
   length=5.5634
   radius=1.1716
   shapeName=dyn_necklace_02_capsule
   center=<3.1685 -0.2859 0.0379>
   rotation=<rot 90 25.673 29.1938>
  }]
 spine_05 [{shapeType=Capsule, length=6.0771, radius=10.7081, shapeName=spine_05_capsule, center=<-4.4145 -1.546 0>}]
}
constraints [
 {
  bone1Name=base_necklace_00
  bone2Name=spine_05
  swing1LimitDegrees=45
  swing1Motion=Free
  swing2LimitDegrees=45
  swing2Motion=Free
  twistLimitDegrees=45
  twistMotion=Free
 }
 {
  bone1Name=clavicle_l
  bone2Name=spine_05
  swing1LimitDegrees=45
  swing1Motion=Free
  swing2LimitDegrees=45
  swing2Motion=Free
  twistLimitDegrees=45
  twistMotion=Free
 }
 {
  bone1Name=clavicle_r
  bone2Name=spine_05
  swing1LimitDegrees=45
  swing1Motion=Free
  swing2LimitDegrees=45
  swing2Motion=Free
  twistLimitDegrees=45
  twistMotion=Free
 }
 {bone1Name=dyn_necklace_00, bone2Name=base_necklace_00, swing1LimitDegrees=15, swing2LimitDegrees=10, twistLimitDegrees=0}
 {bone1Name=dyn_necklace_01, bone2Name=dyn_necklace_00, swing1LimitDegrees=15, swing2LimitDegrees=10, twistLimitDegrees=0}
 {bone1Name=dyn_necklace_02, bone2Name=dyn_necklace_01, swing1LimitDegrees=10, swing2LimitDegrees=16, twistLimitDegrees=0}
 {bone1Name=dyn_necklace_01_collar, bone2Name=dyn_necklace_01, swing1LimitDegrees=20, swing2LimitDegrees=12, twistLimitDegrees=0}
]
