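# Physics asset for the shirt of the medium female FirstClassSecond cosmetic body, set up for secondary motion of the shirt bones.
# Eight base_shirt/dyn_shirt capsule pairs cover front, back, sides and diagonals, alongside larger pelvis and spine_01 capsules for the torso.
# Each dyn bone has a narrow one-axis swing limit against its base, wider at the sides.
# Bases link to spine_01, and spine_01 to pelvis, with free angular motion.
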PhysicsAsset /BRCosmetics/Characters/Player/Female/Medium/Bodies/F_MED_FirstClassSecond/Meshes/F_MED_Shirt_FirstClassSecond_Physics
bodies {
 base_shirt_bk [{shapeType=Capsule, length=0.505, radius=0.505, shapeName=base_shirt_bk_capsule, rotation=<rot 0 -90 0>}]
 base_shirt_bk_l [{shapeType=Capsule, length=0.505, radius=0.505, shapeName=base_shirt_bk_l_capsule, rotation=<rot 0 -90 0>}]
 base_shirt_bk_r [{shapeType=Capsule, length=0.505, radius=0.505, shapeName=base_shirt_bk_r_capsule, rotation=<rot 0 -90 0>}]
 base_shirt_fr [{shapeType=Capsule, length=0.505, radius=0.505, shapeName=base_shirt_fr_capsule, rotation=<rot 0 -90 0>}]
 base_shirt_fr_l [{shapeType=Capsule, length=0.505, radius=0.505, shapeName=base_shirt_fr_l_capsule, rotation=<rot 0 -90 0>}]
 base_shirt_fr_r [{shapeType=Capsule, length=0.505, radius=0.505, shapeName=base_shirt_fr_r_capsule, rotation=<rot 0 -90 0>}]
 base_shirt_l [{shapeType=Capsule, length=0.505, radius=0.505, shapeName=base_shirt_l_capsule, rotation=<rot 0 -90 0>}]
 base_shirt_r [{shapeType=Capsule, length=0.505, radius=0.505, shapeName=base_shirt_r_capsule, rotation=<rot 0 -90 0>}]
 dyn_shirt_bk [{shapeType=Capsule, length=3, radius=1, shapeName=dyn_shirt_bk_capsule, center=<2 0 0>, rotation=<rot 90 0 0>}]
 dyn_shirt_bk_l [{shapeType=Capsule, length=3, radius=1, shapeName=dyn_shirt_bk_l_capsule, center=<2 0 0>, rotation=<rot 90 0 0>}]
 dyn_shirt_bk_r [{shapeType=Capsule, length=3, radius=1, shapeName=dyn_shirt_bk_r_capsule, center=<2 0 0>, rotation=<rot 90 0 0>}]
 dyn_shirt_fr [{shapeType=Capsule, length=3, radius=1, shapeName=dyn_shirt_fr_capsule, center=<2 0 0>, rotation=<rot 90 0 0>}]
 dyn_shirt_fr_l [{shapeType=Capsule, length=3, radius=1, shapeName=dyn_shirt_fr_l_capsule, center=<2 0 0>, rotation=<rot 90 0 0>}]
 dyn_shirt_fr_r [{shapeType=Capsule, length=3, radius=1, shapeName=dyn_shirt_fr_r_capsule, center=<2 0 0>, rotation=<rot 90 0 0>}]
 dyn_shirt_l [{shapeType=Capsule, length=3, radius=1, shapeName=dyn_shirt_l_capsule, center=<2 0 0>, rotation=<rot 90 0 0>}]
 dyn_shirt_r [{shapeType=Capsule, length=3, radius=1, shapeName=dyn_shirt_r_capsule, center=<2 0 0>, rotation=<rot 90 0 0>}]
 pelvis [{shapeType=Capsule, length=10, radius=10, shapeName=pelvis_capsule, center=<0.165 5.348 0.1495>}]
 spine_01 [{shapeType=Capsule, length=8.5, radius=8.5, shapeName=spine_01_capsule, center=<5.3876 -0.9354 0>}]
}
constraints [
 {bone1Name=dyn_shirt_bk, bone2Name=base_shirt_bk, swing1LimitDegrees=10, swing2LimitDegrees=0, twistLimitDegrees=0}
 {bone1Name=dyn_shirt_fr, bone2Name=base_shirt_fr, swing1LimitDegrees=10, swing2LimitDegrees=0, twistLimitDegrees=0}
 {bone1Name=dyn_shirt_l, bone2Name=base_shirt_l, swing1LimitDegrees=20, swing2LimitDegrees=0, twistLimitDegrees=0}
 {bone1Name=dyn_shirt_r, bone2Name=base_shirt_r, swing1LimitDegrees=20, swing2LimitDegrees=0, twistLimitDegrees=0}
 {
  bone1Name=base_shirt_fr
  bone2Name=spine_01
  swing1LimitDegrees=45
  swing1Motion=Free
  swing2LimitDegrees=45
  swing2Motion=Free
  twistLimitDegrees=45
  twistMotion=Free
 }
 {
  bone1Name=base_shirt_bk
  bone2Name=spine_01
  swing1LimitDegrees=45
  swing1Motion=Free
  swing2LimitDegrees=45
  swing2Motion=Free
  twistLimitDegrees=45
  twistMotion=Free
 }
 {
  bone1Name=base_shirt_l
  bone2Name=spine_01
  swing1LimitDegrees=45
  swing1Motion=Free
  swing2LimitDegrees=45
  swing2Motion=Free
  twistLimitDegrees=45
  twistMotion=Free
 }
 {
  bone1Name=base_shirt_r
  bone2Name=spine_01
  swing1LimitDegrees=45
  swing1Motion=Free
  swing2LimitDegrees=45
  swing2Motion=Free
  twistLimitDegrees=45
  twistMotion=Free
 }
 {
  bone1Name=base_shirt_bk_l
  bone2Name=spine_01
  swing1LimitDegrees=45
  swing1Motion=Free
  swing2LimitDegrees=45
  swing2Motion=Free
  twistLimitDegrees=45
  twistMotion=Free
 }
 {bone1Name=dyn_shirt_bk_l, bone2Name=base_shirt_bk_l, swing1LimitDegrees=10, swing2LimitDegrees=0, twistLimitDegrees=0}
 {
  bone1Name=base_shirt_bk_r
  bone2Name=spine_01
  swing1LimitDegrees=45
  swing1Motion=Free
  swing2LimitDegrees=45
  swing2Motion=Free
  twistLimitDegrees=45
  twistMotion=Free
 }
 {bone1Name=dyn_shirt_bk_r, bone2Name=base_shirt_bk_r, swing1LimitDegrees=10, swing2LimitDegrees=0, twistLimitDegrees=0}
 {
  bone1Name=base_shirt_fr_l
  bone2Name=spine_01
  swing1LimitDegrees=45
  swing1Motion=Free
  swing2LimitDegrees=45
  swing2Motion=Free
  twistLimitDegrees=45
  twistMotion=Free
 }
 {bone1Name=dyn_shirt_fr_l, bone2Name=base_shirt_fr_l, swing1LimitDegrees=10, swing2LimitDegrees=0, twistLimitDegrees=0}
 {
  bone1Name=base_shirt_fr_r
  bone2Name=spine_01
  swing1LimitDegrees=45
  swing1Motion=Free
  swing2LimitDegrees=45
  swing2Motion=Free
  twistLimitDegrees=45
  twistMotion=Free
 }
 {bone1Name=dyn_shirt_fr_r, bone2Name=base_shirt_fr_r, swing1LimitDegrees=10, swing2LimitDegrees=0, twistLimitDegrees=0}
 {
  bone1Name=spine_01
  bone2Name=pelvis
  swing1LimitDegrees=45
  swing1Motion=Free
  swing2LimitDegrees=45
  swing2Motion=Free
  twistLimitDegrees=45
  twistMotion=Free
 }
]
